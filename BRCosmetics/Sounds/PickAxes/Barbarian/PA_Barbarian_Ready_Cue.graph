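# Sound cue for the Barbarian pickaxe cosmetic's ready sound.
# A modulator slightly randomizes the volume of a single wave, PA_Barbarian_Ready_01, with the cue at reduced overall volume.
# It uses the melee weapon sound class, close-range projectile attenuation and the shared pickaxe-ready audio property sheet.
# It primes on load and overrides concurrency to prevent new instances once the limit is reached.
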SoundCue /BRCosmetics/Sounds/PickAxes/Barbarian/PA_Barbarian_Ready_Cue
root=Modulator_0
nodes [
 {id=Modulator_0, children [WavePlayer_5], properties {volumeMax=1, volumeMin=0.9}}
 {id=WavePlayer_5, properties {soundWaveAssetPtr=&/PA_Barbarian_Ready_01}}
]
properties {
 attenuationSettings=&Fort_Weapon_Projectile_Close_Athena_Att
 audioPropertiesSheet=/BRCosmetics/Sounds/Settings/PropertySheets/Pickaxe/APS_Cosmetic_Pickaxe_Ready
 bOverrideConcurrency=true
 bPrimeOnLoad=true
 maxDistance=4950
 soundClassObject=&Fort_WeaponMelee
 volumeMultiplier=0.35
 concurrencyOverrides {resolutionRule=PreventNew}
}
